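# Brushed stainless steel material for the Future vehicle body skin, using Default Lit on skeletal meshes.
# A surface-noise texture and a Fresnel term shape metallic and roughness over a flat base color.
# These attributes pass through the shared vehicle damage and de-rez transition functions.
# Anisotropy is then added with a procedural tangent, built from floored, sine-modulated UV1 streaks instead of a direction map, with tangent-space normals off.
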Material /VehicleCosmetics/Bodies/Future/Skins/Stainless/MAT_FutureStainless
graph {
 properties {
  bAutomaticallySetUsageInEditor=false
  bTangentSpaceNormal=false
  bUseMaterialAttributes=true
  bUsedWithMeshDeformer=true
  bUsedWithSkeletalMesh=true
  pixelDepthOffsetMode=PDOM_Legacy
  usedShadingModels=Default Lit
 }
 nodes [
  n1 = TextureSample() {texture=&/Stainless_SurfaceNoise}
  n2 = Multiply(A=@n1.G, B=Constant() {desc=Roughness, r=0.4})
  n3 = Constant() {desc=Roughness Angle, r=2.5}
  n4 = Saturate(None=LinearInterpolate(Alpha=Fresnel(ExponentIn=@n3)) {constA=1, constB=0})
  n5 = SetMaterialAttributes!(
   MaterialAttributes
   Base Color=Constant() {desc=Color, r=0.2}
   Metallic=LinearInterpolate(A=Constant() {r=0.8}, Alpha=@n1.G) {constB=0.4}
   Roughness=LinearInterpolate(A=LinearInterpolate(B=@n2, Alpha=@n4) {constA=0.03, constB=0.48}, Alpha=Saturate(None=@n1.G)) {constA=0.01, constB=0.3}
   Normal
  ) {attributeSetTypes [
    #1
    #2
    #3
    #4
   ]}
  n6 = Multiply(A=TextureCoordinate() {coordinateIndex=1, vTiling=2}, B=Constant() {r=2048})
  n7 = Multiply(A=@n6) {constB=2}
  n8 = Multiply(A=ComponentMask(None=Floor(None=@n7)) {g=true}) {constB=0.125}
  n9 = Multiply(A=@n6) {constB=0.0125}
  n10 = ComponentMask(None=Sine(None=Add(A=@n8, B=@n9)) {period=0.5}) {g=true}
  n11 = Abs(None=@n10)
  n12 = Multiply(A=LinearInterpolate(A=@n11, B=Multiply(A=@n11) {constB=0.1}, Alpha=Ceil(None=@n10)), B=Constant() {desc=Aniso Intensity, r=0.01}) {constB=0.02, desc=R}
  n13 = Constant() {desc=G, r=1}
  SetMaterialAttributes!(
   MaterialAttributes=MaterialFunctionCall(In=MaterialFunctionCall(Material=@n5) {materialFunction=&MF_BaseVehicle_Damage}.Result) {materialFunction=&MF_VehicleDeRez_Transition}.Result
   Anisotropy=Constant() {desc=Aniso Angle, r=0.9}
   Tangent=AppendVector(A=AppendVector(A=@n12, B=@n13), B=Constant() {desc=B, r=1})
  ) {attributeSetTypes [#5, #6]}
 ]
}
